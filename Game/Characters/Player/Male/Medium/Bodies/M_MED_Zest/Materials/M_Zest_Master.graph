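# Master material for the Zest player character body, using Cloth shading with material attributes and supporting skeletal, morph and clothing meshes.
# Base character PreFX feeds the Marshmello weapon-skin effect, driven by selectable channels of a SkinFX mask, then anime shading.
# The shaded result passes through Peppermint's hair-puff world-position-offset function into character PostFX.
# A separate branch feeds a mobile toon-outline function with no shown consumer.
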Material /Game/Characters/Player/Male/Medium/Bodies/M_MED_Zest/Materials/M_Zest_Master
graph {
 properties {
  bAutomaticallySetUsageInEditor=false
  bUseMaterialAttributes=true
  bUsedWithClothing=true
  bUsedWithMeshDeformer=true
  bUsedWithMorphTargets=true
  bUsedWithSkeletalMesh=true
  pixelDepthOffsetMode=PDOM_Legacy
  shadingModel=MSM_Cloth
  usedShadingModels=Cloth
 }
 nodes [
  n1 = TextureSampleParameter2D() {parameterName=SkinFX_Mask}
  n2 = MaterialFunctionCall() {materialFunction=&BaseFN_CharacterFunction_PreFX}
  n3 = MaterialFunctionCall!(
   Material=MaterialFunctionCall!(
    Alpha=StaticComponentMaskParameter(None=@n1.RGBA) {defaultR=true, group=Skin FX, parameterName=Marshmello Channel}
    Offset=StaticComponentMaskParameter(None=@n1.RGBA) {defaultG=true, group=Skin FX, parameterName=Marshmello Offset Channel}
    EQ Low=StaticComponentMaskParameter(None=@n1.RGBA) {defaultR=true, group=Skin FX, parameterName=Marshmello Low Channel}
    EQ High=StaticComponentMaskParameter(None=@n1.RGBA) {defaultG=true, group=Skin FX, parameterName=Marshmello High Channel}
    Material=@n2.Result
   ) {materialFunction=/Game/Weapons/WeaponSkins/MaterialFunction/Marshmello}.Result
   LitDiffuseOverride
   SSC_AlphaOverride
   Tint Mask
   DF InkLine Visibility Mask
   DF Inkline Injection
   DF Inkline Color Override
   DF Inkline Color Override Mask
   CustomUVs
  ) {materialFunction=/Game/Packages/Fortress_SharedMaterials/Anime_Master_Material/MF_AnimeShading}
  MaterialFunctionCall!(Material=@n3.Result) {materialFunction=/Game/Packages/Fortress_SharedMaterials/Anime_Master_Material/MF_ToonOutline_Mobile}
  MaterialFunctionCall(
   AO=@n2.AO
   Distance Blend=@n2.AO
   Fresnel=@n2.AO
   Gmap=@n2.AO
   Pre Skinned Local Position=@n2.AO
   Material=MaterialFunctionCall!(In=@n3.Result) {materialFunction=/Game/Characters/Player/Female/Medium/Bodies/F_MED_Peppermint/Materials/MF_HairPuffWPO}.Result
  ) {materialFunction=&BaseFN_CharacterFunction_PostFX}
 ]
}
